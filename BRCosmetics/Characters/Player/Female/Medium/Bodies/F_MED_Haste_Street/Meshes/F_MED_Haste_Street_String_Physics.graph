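# Physics asset for the F_MED_Haste_Street female body cosmetic, defining bodies and joints for secondary motion on two strings and a grenade.
# Left and right three-capsule string chains hang from dyn_string_parent, a compound body of three large capsules and a box.
# Swing limits tighten down the right chain but stay uniform on the left, with tight twist limits on both.
# The grenade capsule attaches separately to a box on the right thigh twist bone, with asymmetric swing limits.
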PhysicsAsset /BRCosmetics/Characters/Player/Female/Medium/Bodies/F_MED_Haste_Street/Meshes/F_MED_Haste_Street_String_Physics
bodies {
 dyn_grenade [{shapeType=Capsule, length=7.3551, radius=3.9119, center=<3.5046 -0.3003 1.0558>, rotation=<rot -2.6524 -90 3.8034>}]
 dyn_string_l_01 [{shapeType=Capsule, length=2.0868, radius=0.505, center=<1.499 0.0248 0.1504>, rotation=<rot 90 -179.8419 -179.8415>}]
 dyn_string_l_02 [{shapeType=Capsule, length=2.6343, radius=0.505, center=<1.0588 -0.0137 0.1504>, rotation=<rot 90 -2.5304 -2.5302>}]
 dyn_string_l_03 [{shapeType=Capsule, length=3.406, radius=0.5354, center=<1.281 0.0058 0.1502>, rotation=<rot 90 1.8986 1.8989>}]
 dyn_string_parent [
  {shapeType=Capsule, length=11.5, radius=5.75, center=<5.1179 8.4847 -4.6808>, rotation=<rot -84.9275 -180 180>}
  {shapeType=Capsule, length=11.5, radius=5.75, center=<5.2356 -7.0949 -3.5776>, rotation=<rot -84.9312 -180 180>}
  {shapeType=Capsule, length=12.1575, radius=9.5925, center=<9.0175 0.5204 -7.4369>, rotation=<rot -75.8896 -171.0962 81.0143>}
  {shapeType=Box, extentX=16.5, extentY=17, extentZ=7.75, center=<-0.5965 0.0003 -1.8244>, rotation=<rot 0.5361 0 0>}
 ]
 dyn_string_r_01 [{shapeType=Capsule, length=1.9618, radius=0.63, center=<1.4841 0.1948 0.1503>, rotation=<rot 90 -1.1935 -6.1935>}]
 dyn_string_r_02 [{shapeType=Capsule, length=2.5469, radius=0.63, center=<1.1956 0.2169 0.0294>, rotation=<rot 90 -1.7899 -6.79>}]
 dyn_string_r_03 [{shapeType=Capsule, length=3.115, radius=0.6409, center=<1.2692 0.1601 -0.0001>, rotation=<rot 90 -178.2101 176.7898>}]
 thigh_twist_01_r [{shapeType=Box, extentX=7.5, extentY=7.5, extentZ=7.5}]
}
constraints [
 {bone1Name=dyn_string_r_01, bone2Name=dyn_string_parent, swing1LimitDegrees=25, swing2LimitDegrees=25, twistLimitDegrees=5}
 {bone1Name=dyn_string_r_02, bone2Name=dyn_string_r_01, swing1LimitDegrees=20, swing2LimitDegrees=20, twistLimitDegrees=5}
 {bone1Name=dyn_string_r_03, bone2Name=dyn_string_r_02, swing1LimitDegrees=15, swing2LimitDegrees=15, twistLimitDegrees=5}
 {bone1Name=dyn_string_l_01, bone2Name=dyn_string_parent, swing1LimitDegrees=15, swing2LimitDegrees=15, twistLimitDegrees=5}
 {bone1Name=dyn_string_l_02, bone2Name=dyn_string_l_01, swing1LimitDegrees=15, swing2LimitDegrees=15, twistLimitDegrees=5}
 {bone1Name=dyn_string_l_03, bone2Name=dyn_string_l_02, swing1LimitDegrees=15, swing2LimitDegrees=15, twistLimitDegrees=5}
 {bone1Name=dyn_grenade, bone2Name=thigh_twist_01_r, swing1LimitDegrees=25, swing2LimitDegrees=10, twistLimitDegrees=3}
]
